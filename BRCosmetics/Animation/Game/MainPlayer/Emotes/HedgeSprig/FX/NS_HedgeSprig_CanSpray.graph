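# Niagara system for a can spray effect in the HedgeSprig player emote.
# It holds one emitter, SparklesLWE (likely lightweight, by its name), with no renderers or emitter modules listed here.
# System scripts use SystemState plus a spawn-time variable assignment, and a time-of-day color dynamic input is among the dependencies.
# A User.NormalizedNotifyProgress float is exposed, apparently for animation-notify progress.
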NiagaraSystem /BRCosmetics/Animation/Game/MainPlayer/Emotes/HedgeSprig/FX/NS_HedgeSprig_CanSpray
dependencies {
 usedDataInterfaces []
 usedDynamicInputs [&FNTimeOfDayColor]
 usedModules [
  &SystemState
  &self:SystemSpawnScript.SystemScriptSource.SystemScriptGraph.NiagaraNodeAssignment_0.SetVariables_#1
 ]
}
emitters {SparklesLWE {rendererClasses [], renderers []}}
summary {
 systemName=NS_HedgeSprig_CanSpray
 emitters [{emitterName=SparklesLWE, rendererClasses []}]
 userVariables [User.NormalizedNotifyProgress {type=NiagaraFloat, defaultValue=0, description=""}]
}
